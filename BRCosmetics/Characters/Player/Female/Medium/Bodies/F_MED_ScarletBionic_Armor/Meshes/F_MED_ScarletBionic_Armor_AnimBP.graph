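# Post-process style cosmetic AnimBlueprint for the female medium ScarletBionic armor, copying the attached parent mesh's pose and adding armor-specific motion.
# Shoulder bones blend to preset transforms during skydiving and parachuting.
# Spike bones get a mirrored real-time sine bob from the event graph, then spring-based AnimDynamics simulated relative to the foot bones.
# The armor's Control Rig runs last, before output.
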AnimBlueprint /BRCosmetics/Characters/Player/Female/Medium/Bodies/F_MED_ScarletBionic_Armor/Meshes/F_MED_ScarletBionic_Armor_AnimBP
target_skeleton=&/F_MED_ScarletBionic_Armor_Skeleton
graphs [
 AnimGraph {nodes [
   n1 = Get bIsSkyDiving()
   n2 = "Transform (Modify) Bone - Bone: dyn_shoulder_l"!(
    ComponentPose=Local To Component(LocalPose=Copy Pose From Mesh() {bUseAttachedParent=true})
    bAlphaBoolEnabled:bool=@n1
    Translation="-0.000,1.748,1.337"
    Rotation="-0.000,-0.000,8.788"
    Scale="1,1,1"
   ) -> (Pose) {
    alphaInputType=Bool
    rotationMode=BMM_Replace
    rotationSpace=BCS_BoneSpace
    translationMode=BMM_Replace
    translationSpace=BCS_BoneSpace
    alphaBoolBlend {blendInTime=0.3, blendOption=Linear, blendOutTime=0.3}
    boneToModify {boneName=dyn_shoulder_l}
    rotation=<rot 0 0 8.788>
    translation {y=1.748, z=1.337}
   }
   n3 = "Transform (Modify) Bone - Bone: dyn_shourder_r"!(ComponentPose=@n2, bAlphaBoolEnabled:bool=@n1, Translation="0,-1.748,-1.337", Rotation="0,0,8.788", Scale="1,1,1") -> (Pose) {
    alphaInputType=Bool
    rotationMode=BMM_Replace
    rotationSpace=BCS_BoneSpace
    translationMode=BMM_Replace
    translationSpace=BCS_BoneSpace
    alphaBoolBlend {blendInTime=0.3, blendOption=Linear, blendOutTime=0.3}
    boneToModify {boneName=dyn_shourder_r}
    rotation {roll=8.788}
    translation {y=-1.748, z=-1.337}
   }
   n4 = Get bIsParachuteOpened()
   n5 = "Transform (Modify) Bone - Bone: dyn_shoulder_l"(ComponentPose=@n3, bAlphaBoolEnabled=@n4, Translation="4.257,-0.557,-1.918", Rotation="1.958,-12.792,8.567", Scale="1,1,1") {
    alphaInputType=Bool
    rotationMode=BMM_Replace
    rotationSpace=BCS_BoneSpace
    translationMode=BMM_Replace
    translationSpace=BCS_BoneSpace
    alphaBoolBlend {blendInTime=0.3, blendOption=Linear, blendOutTime=0.3}
    boneToModify {boneName=dyn_shoulder_l}
    rotation=<rot 1.958 -12.792 8.567>
    translation=<4.257 -0.557 -1.918>
   }
   n6 = "Transform (Modify) Bone - Bone: dyn_shourder_r"(ComponentPose=@n5, bAlphaBoolEnabled=@n4, Translation="-4.257,0.440,1.181", Rotation="1.958,-12.792,8.567", Scale="1,1,1") {
    alphaInputType=Bool
    rotationMode=BMM_Replace
    rotationSpace=BCS_BoneSpace
    translationMode=BMM_Replace
    translationSpace=BCS_BoneSpace
    alphaBoolBlend {blendInTime=0.3, blendOption=Linear, blendOutTime=0.3}
    boneToModify {boneName=dyn_shourder_r}
    rotation=<rot 1.958 -12.792 8.567>
    translation=<-4.257 0.44 1.181>
   }
   n7 = "Transform (Modify) Bone - Bone: dyn_spike_l"!(
    ComponentPose=@n6
    bAlphaBoolEnabled:bool=True
    Translation=Get ambient_motion_spike_l() -> (ambient_motion_spike_l:Vector)
    Rotation="0,0,0"
    Scale="1,1,1"
   ) -> (Pose) {
    alphaInputType=Bool
    rotationMode=BMM_Additive
    rotationSpace=BCS_BoneSpace
    translationMode=BMM_Additive
    translationSpace=BCS_BoneSpace
    alphaBoolBlend {blendInTime=0.3, blendOption=Linear, blendOutTime=0.3}
    boneToModify {boneName=dyn_spike_l}
   }
   n8 = "Transform (Modify) Bone - Bone: dyn_spike_r"!(
    ComponentPose=@n7
    bAlphaBoolEnabled:bool=True
    Translation=Get ambient_motion_spike_r() -> (ambient_motion_spike_r:Vector)
    Rotation="0,0,0"
    Scale="1,1,1"
   ) -> (Pose) {
    alphaInputType=Bool
    rotationMode=BMM_Additive
    rotationSpace=BCS_BoneSpace
    translationMode=BMM_Additive
    translationSpace=BCS_BoneSpace
    alphaBoolBlend {blendInTime=0.3, blendOption=Linear, blendOutTime=0.3}
    boneToModify {boneName=dyn_spike_r}
   }
   n9 = Anim Dynamics - Bone: dyn_spike_l!(
    ComponentPose=@n8
    Alpha=0.7
    GravityOverride:Vector="0,0,50"
    ComponentLinearAccScale:Vector="0.1,0.1,0.1"
    ComponentLinearVelScale:Vector="0.1,0.1,0.1"
   ) -> (Pose) {
    alpha=0
    angularBiasOverride=1
    angularDampingOverride=0.75
    angularSpringConstant=20
    bGravityOverrideInSimSpace=true
    bLinearSpring=true
    bOverrideLinearDamping=true
    bUseGravityOverride=true
    bUsePlanarLimit=false
    bUseSphericalLimits=true
    gravityScale=0
    linearDampingOverride=1
    linearSpringConstant=50
    simulationSpace=BoneRelative
    alphaBoolBlend {blendOption=Linear}
    boundBone {boneName=dyn_spike_l}
    componentAppliedLinearAccClamp=<80 80 80>
    gravityOverride {z=20}
    physicsBodyDefinitions [{
      sphereCollisionRadius=0
      boundBone {boneName=dyn_spike_l}
      boxExtents=<5 5 5>
      constraintSetup {angularLimitsMax=<0 0 0>, angularLimitsMin=<0 0 0>, angularTarget=<0 0 0>, linearAxesMax=<3 0 3>, linearAxesMin=<-3 -3 -3>}
      localJointOffset=<0 0 0>
     }]
    relativeSpaceBone {boneName=foot_l}
    sphericalLimits [{limitRadius=5, limitType=Inner, drivingBone {boneName=dyn_spike_l}, sphereLocalOffset=<0 0 4.4>}]
   }
   n10 = Anim Dynamics - Bone: dyn_spike_r!(
    ComponentPose=@n9
    Alpha=0.7
    GravityOverride:Vector="0,0,50"
    ComponentLinearAccScale:Vector="0.1,0.1,0.1"
    ComponentLinearVelScale:Vector="0.1,0.1,0.1"
   ) -> (Pose) {
    alpha=0
    angularBiasOverride=1
    angularDampingOverride=0.75
    angularSpringConstant=20
    bGravityOverrideInSimSpace=true
    bLinearSpring=true
    bOverrideLinearDamping=true
    bUseGravityOverride=true
    bUsePlanarLimit=false
    bUseSphericalLimits=true
    gravityScale=0
    linearDampingOverride=1
    linearSpringConstant=50
    simulationSpace=BoneRelative
    alphaBoolBlend {blendOption=Linear}
    boundBone {boneName=dyn_spike_r}
    componentAppliedLinearAccClamp=<80 80 80>
    gravityOverride {z=20}
    physicsBodyDefinitions [{
      sphereCollisionRadius=0
      boundBone {boneName=dyn_spike_r}
      boxExtents=<5 5 5>
      constraintSetup {angularLimitsMax=<0 0 0>, angularLimitsMin=<0 0 0>, angularTarget=<0 0 0>, linearAxesMax=<3 0 3>, linearAxesMin=<-3 -3 -3>}
      localJointOffset=<0 0 0>
     }]
    relativeSpaceBone {boneName=foot_r}
    sphericalLimits [{limitRadius=5, limitType=Inner, drivingBone {boneName=dyn_spike_r}, sphereLocalOffset=<0 0 4.4>}]
   }
   n11 = Control Rig(Source=Component To Local(ComponentPose=@n10)) {
    alphaBoolBlend {blendOption=Linear}
    controlRigAssetReference {blueprintRigClass=&/F_MED_ScarletBionic_Armor_CtrlRig.F_MED_ScarletBionic_Armor_CtrlRig_C}
   }
   Output Pose(Result=@n11) {layerGroup=None}
  ]}
 EventGraph {nodes [
   Event BlueprintUpdateAnimation()
   n1 = GetRealTimeSeconds()
   n2 = float * float(A=1.36, B=0.75)
   n3 = float * float(A="Sin (Radians)"(A=float * float(A=@n2, B=@n1)), B=1.2)
   > Set ambient_motion_spike_l(ambient_motion_spike_l:Vector=MakeVector(X=@n3, Y=0)) -> (Output_Get:Vector)
   > Set ambient_motion_spike_r(ambient_motion_spike_r:Vector=MakeVector(X=float * float(A=@n3, B=-1) {is=title K2Node_PromotableOperator}, Y=0)) -> (Output_Get:Vector)
  ]}
]
variables [
 ambient_motion_spike_l {type=Vector}
 ambient_motion_spike_r {type=Vector}
]
